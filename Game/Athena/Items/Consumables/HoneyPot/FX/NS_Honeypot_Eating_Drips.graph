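# Niagara system for honey drips, apparently shown while eating the Honey Pot consumable.
# Its one sprite emitter, Drip_Drop, continuously spawns short-lived, randomly sized droplets in a small sphere, pushes them outward from a point and pulls them down with gravity.
# A 2D curve quickly grows each sprite, then shrinks it to zero.
# Color and alpha lerps resolve to constants, and the camera offset module is disabled.
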NiagaraSystem /Game/Athena/Items/Consumables/HoneyPot/FX/NS_Honeypot_Eating_Drips
dependencies {
 usedDataInterfaces [&NiagaraDataInterfaceVector2DCurve]
 usedDynamicInputs []
 usedModules [&SystemState]
}
emitters {Drip_Drop {
  emitterUpdateScript [
   &EmitterState()
   &SpawnRate(SpawnRate=25, Use Spawn Probability=false)
  ]
  particleSpawnScript [
   /Niagara/Modules/Spawn/Initialization/InitializeParticle(
    Lifetime=&UniformRangedFloat(Minimum=0.3, Maximum=0.8)
    Module.Write Color=true
    Module.Write Lifetime=true
    Module.Write Mass=true
    Module.Write Material Random=true
    Module.Write Position=true
    Module.Write Scale=false
    Module.Write SpriteRotation=false
    Module.Write SpriteSize=true
    Sprite Size=&UniformRangedVector2D(Minimum=<7.5 10>, Maximum=<10 17.5>)
   )
   /Niagara/Modules/Spawn/Location/SphereLocation(MaskSpawnGroup=false, Module.Override Randomness=false, Sphere Radius=7, Surface Only=false)
   &AddVelocityFromPoint(Module.Use Falloff Distance=false, Velocity Strength=&UniformRangedFloat(Minimum=15, Maximum=35))
  ]
  particleUpdateScript [
   &CameraOffset(Camera Offset Amount=12) {enabled=false}
   &ParticleState()
   &GravityForce(Gravity=<0 0 -400>)
   &SolveForcesAndVelocity(Clamp Velocity=false, Limit Acceleration=false)
   &ScaleColor(
    Scale Alpha=&Lerp_Float(A=1, B=1, Alpha=$Particles.DistanceAlpha)
    Scale RGB=&MakeVectorFromLinearColorRGB(Color=&LerpLinearColors(StartColor=<rgba 0.9 0.9 0.9 1>, EndColor=<rgba 0.1396 1.4602 1.5 1>, LerpFactor=0))
    ScaleA=true
    ScaleRGB=true
   )
   &ScaleSpriteSize(Scale Factor=&Vector2DFromCurve(
     Vector2Curve=DataInterface() {propertyValues {
       XCurve {keys [
         {arriveTangent=5.9728, interpMode=RCIM_Cubic, leaveTangent=5.9728, tangentMode=RCTM_User, time=0, value=0.1}
         {arriveTangent=-0.1, interpMode=RCIM_Cubic, leaveTangent=-0.1, tangentMode=RCTM_User, time=0.18, value=0.99}
         {interpMode=RCIM_Cubic, tangentMode=RCTM_User, time=1, value=0}
        ]}
       YCurve {keys [
         {arriveTangent=6.2715, interpMode=RCIM_Cubic, leaveTangent=6.2715, tangentMode=RCTM_User, time=0, value=0.1}
         {arriveTangent=-0.1001, interpMode=RCIM_Cubic, leaveTangent=-0.1001, tangentMode=RCTM_User, time=0.18, value=0.99}
         {interpMode=RCIM_Cubic, tangentMode=RCTM_User, time=0.9989, value=0}
        ]}
      }}
     ScaleCurve=<1 1>
    ))
  ]
  rendererClasses [&NiagaraSpriteRendererProperties]
  renderers [{rendererClass=&NiagaraSpriteRendererProperties}]
 }}
summary {
 systemName=NS_Honeypot_Eating_Drips
 emitters [{emitterName=Drip_Drop, rendererClasses [&NiagaraSpriteRendererProperties]}]
 userVariables []
}
